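# Editor utility component that loads or unloads the Asteria lobby island and landscape data layers in the editor.
# setDataLayerVisibility resolves the assets in DataLayersToLoad to layer instances through the Data Layer Editor Subsystem.
# True loads each instance and makes it visible, while false only unloads it.
# The delayed FakeConstruct hook calls an empty function, and getDataLayerVisibility always returns false.
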EditorUtilityBlueprint /Game/Blueprints/Blutilities/ForLDs/WorldPartitionTools/LobbyIslandLoader/ActorComp_LobbyIslandLoader
graphs [
 EventGraph {nodes [
   Event FakeConstruct() {is=title K2Node_Event}
   > Delay!(execute, WorldContextObject:Object Reference, Duration=0.05) -> (then)
   > fakeConstruct_internal()
   n1 = Event setDataLayerVisibility!() -> (OutputDelegate, then, isVisible:bool) {is=title K2Node_Event}
   > setLobbyIslandVisibility_Internal!(execute, self, isVisible:bool=@n1.isVisible) -> (then)
   Event Begin Play()
   Event Tick!() -> (OutputDelegate, then, DeltaSeconds:float)
  ]}
 fakeConstruct_internal {nodes [fakeConstruct_internal() {is=title K2Node_FunctionEntry}]}
 getDataLayerVisibility {nodes [
   getDataLayerVisibility() {is=title K2Node_FunctionEntry}
   > Return Node!(execute, isVisible:bool=false)
  ]}
 setLobbyIslandVisibility_Internal {nodes [
   n1 = setLobbyIslandVisibility_Internal!() -> (then, isVisible:bool=~) {is=title K2Node_FunctionEntry}
   > n2 = For Each Loop(Array:Array of Data Layer Asset Object References=Get DataLayersToLoad() -> (DataLayersToLoad:Array of Data Layer Asset Object References)) -> (Array Element:Data Layer Asset Object Reference)
   n3 = For Each Loop(
    Exec=@n2.Completed
    Array:Array of Data Layer Instance Object References=Get l_instances() -> (l_instances:Array of Data Layer Instance Object References)
   ) -> (Array Element:Data Layer Instance Object Reference)
   > Set l_instance(l_instance:Data Layer Instance Object Reference=@n3) -> (Output_Get:Data Layer Instance Object Reference)
   > n4 = Branch(Condition=@n1)
   n5 = Get DataLayerEditorSubsystem() -> (ReturnValue:Data Layer Editor Subsystem Object Reference) {is=title K2Node_GetEditorSubsystem}
   n6 = Get l_instance() -> (l_instance:Data Layer Instance Object Reference)
   > SetDataLayerIsLoadedInEditor!(
    execute
    self:Data Layer Editor Subsystem Object Reference=@n5
    DataLayer:Data Layer Instance Object Reference=@n6
    bIsLoadedInEditor:bool=true
    bIsFromUserChange:bool=true
   ) -> (then, ReturnValue:bool)
   > SetDataLayerVisibility!(execute, self:Data Layer Editor Subsystem Object Reference=@n5, DataLayer:Data Layer Instance Object Reference=@n6, bIsVisible:bool=true) -> (then)
   SetDataLayerIsLoadedInEditor(execute=@n4.else, self=@n5, DataLayer=@n6, bIsLoadedInEditor=false, bIsFromUserChange=true)
   Add(
    execute=@n2.LoopBody
    TargetArray:Array of Data Layer Instance Object References=Get l_instances()
    NewItem:"Data Layer Instance Object Reference (by ref)"=GetDataLayerInstance!(self:Data Layer Editor Subsystem Object Reference=Get DataLayerEditorSubsystem(), DataLayerAsset:Data Layer Asset Object Reference=@n2) -> (ReturnValue:Data Layer Instance Object Reference)
   ) {is=title K2Node_CallArrayFunction}
  ]}
]
variables [DataLayersToLoad {
  type="array<DataLayerAsset>"
  default [/Game/Athena/Asteria/DataLayers/Asteria_DL_LobbyIsland, /Game/Athena/Asteria/DataLayers/Asteria_DL_LobbyIslandLandscape]
 }]
