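; BP_LobbyIslandLoaderActor is an editor-utility actor in the level designers' World Partition tools that shows or hides the lobby island.
; LoadLobbyIsland and HideLobbyIsland send the ILobbyIslandLoader SetDataLayerVisibility message, with true or false, to its GetActorComp_LobbyIslandLoader component.
; The construction script sends FakeConstruct to the same component, so the actor is a thin wrapper and the data layer logic lives in the component.
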(fn HideLobbyIsland
  (Class|ILobbyIslandLoader|SetDataLayerVisibility(Message) (Variables|Default|GetActorComp_LobbyIslandLoader) false))

(fn LoadLobbyIsland
  (Class|ILobbyIslandLoader|SetDataLayerVisibility(Message) (Variables|Default|GetActorComp_LobbyIslandLoader) true))

(fn ConstructionScript
  (Class|ILobbyIslandLoader|FakeConstruct(Message) (Variables|Default|GetActorComp_LobbyIslandLoader)))
