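; Blueprint interface for the lobby island loader, part of the World Partition editor tools for level designers.
; It declares FakeConstruct, a getter that returns data layer visibility as a boolean, and a setter that takes an isVisible boolean.
; It holds only signatures, so implementers supply the construct and visibility behavior.
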(fn FakeConstruct)

(fn getDataLayerVisibility
  (return false))

(fn setDataLayerVisibility
  (params (isVisible :type Boolean)))
